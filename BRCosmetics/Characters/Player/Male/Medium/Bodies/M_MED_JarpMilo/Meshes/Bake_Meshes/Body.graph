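# Default Lit material named Body in the baked meshes folder of the M_MED_JarpMilo male medium player body, apparently generated by an Interchange mesh import.
# It uses no textures or parameters, only constants: matching dark grey ambient and diffuse, black specular, and a high shininess.
# These feed MF_PhongToMetalRoughness, which converts Phong-style inputs to metal-roughness shading, giving a flat placeholder-like look.
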Material /BRCosmetics/Characters/Player/Male/Medium/Bodies/M_MED_JarpMilo/Meshes/Bake_Meshes/Body
graph {
 properties {assetImportData="<self>:InterchangeAssetImportData_0", usedShadingModels=Default Lit}
 nodes [
  Constant3Vector() {constant=<rgba 0 0 0 1>}
  MaterialFunctionCall!(
   AmbientColor=Constant3Vector() {constant=<rgba 0.1903 0.1903 0.1903 1>}
   DiffuseColor=Constant3Vector() {constant=<rgba 0.1903 0.1903 0.1903 1>}
   SpecularColor=Constant3Vector() {constant=<rgba 0 0 0 1>}
   Shininess=Constant() {r=1024}
  ) {materialFunction=/InterchangeAssets/Functions/MF_PhongToMetalRoughness}
 ]
}
